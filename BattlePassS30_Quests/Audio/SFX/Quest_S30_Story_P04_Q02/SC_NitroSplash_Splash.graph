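# Sound cue for the Nitro Splash impact in a Season 30 story quest, combining glass smash and splash sounds.
# A mixer layers a louder, pitched-down random pick of five glass smashes with a random pick of five splashes, delayed briefly so the smash leads.
# It plays in the gadgets sound class with an attenuation asset and its own concurrency limit that stops the oldest instance.
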SoundCue /BattlePassS30_Quests/Audio/SFX/Quest_S30_Story_P04_Q02/SC_NitroSplash_Splash
root=Mixer_0
nodes [
 {id=Delay_0, children [Modulator_2], properties {delayMax=0.07, delayMin=0.07}}
 {id=Mixer_0, children [Delay_0,Modulator_0], properties {inputVolume [1,1.4]}}
 {id=Modulator_0, children [Random_0], properties {pitchMax=0.95, pitchMin=0.85}}
 {id=Modulator_2, children [Random_1]}
 {
  id=Random_0
  children [WavePlayer_1,WavePlayer_2,WavePlayer_3,WavePlayer_4,WavePlayer_8]
  properties {bShouldExcludeFromBranchCulling=true, weights [1,1,1,1,1]}
 }
 {
  id=Random_1
  children [WavePlayer_0, WavePlayer_5, WavePlayer_6, WavePlayer_7, WavePlayer_12]
  properties {bShouldExcludeFromBranchCulling=true, weights [1,1,1,1,1]}
 }
 {id=WavePlayer_0, properties {soundWaveAssetPtr=&/NitroSplash_Splash_01}}
 {id=WavePlayer_1, properties {soundWaveAssetPtr=&/NitroSplash_GlassSmash_01}}
 {id=WavePlayer_12, properties {soundWaveAssetPtr=&/NitroSplash_Splash_05}}
 {id=WavePlayer_2, properties {soundWaveAssetPtr=&/NitroSplash_GlassSmash_02}}
 {id=WavePlayer_3, properties {soundWaveAssetPtr=&/NitroSplash_GlassSmash_03}}
 {id=WavePlayer_4, properties {soundWaveAssetPtr=&/NitroSplash_GlassSmash_04}}
 {id=WavePlayer_5, properties {soundWaveAssetPtr=&/NitroSplash_Splash_02}}
 {id=WavePlayer_6, properties {soundWaveAssetPtr=&/NitroSplash_Splash_03}}
 {id=WavePlayer_7, properties {soundWaveAssetPtr=&/NitroSplash_Splash_04}}
 {id=WavePlayer_8, properties {soundWaveAssetPtr=&/NitroSplash_GlassSmash_05}}
]
properties {
 attenuationSettings=&/Att_NitroSplash_Splash
 bOverrideConcurrency=true
 maxDistance=6500
 soundClassObject=&Fort_Gadgets
 volumeMultiplier=0.8
 attenuationOverrides {
  bAttenuate=false
  bAttenuateWithLPF=true
  falloffDistance=4300
  lPFFrequencyAtMax=1000
  lPFRadiusMax=3200
  lPFRadiusMin=700
  stereoSpread=1200
  attenuationShapeExtents {x=300}
 }
 concurrencyOverrides {maxCount=3, resolutionRule=StopOldest, volumeScaleAttackTime=0.1}
}
